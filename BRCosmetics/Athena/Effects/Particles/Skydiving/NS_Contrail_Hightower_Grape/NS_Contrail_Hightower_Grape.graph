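# Niagara skydiving contrail cosmetic (Hightower Grape) that trails gold-orange energy from four points on the player's skeletal mesh.
# Ribbons, curl-noise sparkle dust and spinning mesh-and-sprite energy bits stream upward, with added velocity stretched in Z by the emitter's Dive attribute.
# Soft glow sprites and pulsing lights stay pinned to the sampled points each frame.
# Looping bursts of small rotating bramble meshes also follow those points, scaling in and out.
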NiagaraSystem /BRCosmetics/Athena/Effects/Particles/Skydiving/NS_Contrail_Hightower_Grape/NS_Contrail_Hightower_Grape
dependencies {
 usedDataInterfaces [&NiagaraDataInterfaceColorCurve, &NiagaraDataInterfaceCurve]
 usedDynamicInputs []
 usedModules [&SystemState]
}
emitters {
 Brambles001 {
  emitterUpdateScript [
   &EmitterState(
    Inactive Response=::NewEnumerator1
    Life Cycle Mode=::NewEnumerator1
    Loop Duration=&UniformRangedFloat(Minimum=0.8, Maximum=1.1)
    Recalculate Duration Each Loop=true
    UseLoopDelay=false
   )
   &SpawnBurst_Instantaneous(Spawn Count=12, Use Spawn Probability=false)
  ]
  particleSpawnScript [
   &FN_SetSpawnGroup_v2(Number of Spawn Groups=4)
   /Niagara/Modules/Spawn/Initialization/InitializeParticle(
    Lifetime=&UniformRangedFloat(Minimum=2.5, Maximum=3.5)
    Mesh Scale=&UniformRangedVector(Minimum=<0.07 0.07 0.07>, Maximum=<0.08 0.08 0.08>)
    Module.Write Color=true
    Module.Write Lifetime=true
    Module.Write Mass=true
    Module.Write Material Random=true
    Module.Write Position=true
    Module.Write Scale=true
    Module.Write SpriteRotation=false
    Module.Write SpriteSize=false
    Position=&Add_Vector(A=/Niagara/DynamicInputs/Helpers/SimulationPosition(), B=<0 0 0>)
    Sprite Size=<5 5>
   )
   /Niagara/Modules/Spawn/Orientation/V2/InitialMeshOrientation(
    Orientation Vector=&UniformRangedVector(Minimum=<0 0 0>, Maximum=<1 1 1>)
    Rotation=&UniformRangedVector(Minimum=<0 0 0>, Maximum=<360 0 0>)
    Use Orientation Vector=true
    Use Rotation Vector=true
   )
   /SetVariables(Particles.InitialRotationRate:NiagaraFloat=&UniformRangedFloat(Minimum=0.1, Maximum=0.2)) {isSetParameters=true}
   /SetVariables(Particles.ScaleFlipper:NiagaraFloat=&UniformAOrBFloat(A=-1)) {isSetParameters=true}
   /SetVariables(Particles.InitialLocation_Offset:Vector3f=&UniformRangedVector(Minimum=<-2 -2 -2>, Maximum=<2 2 2>)) {isSetParameters=true}
  ]
  particleUpdateScript [
   &SkeletalMeshLocation(
    Bone / Socket Index=$Particles.SpawnGroup
    Bone Sampling Mode=::NewEnumerator4
    Inherit Velocity=false
    MaskSpawnGroup=false
    Sampled Position Offset=&Multiply_Vector(A=<1 1 1>, B=$Particles.InitialLocation_Offset)
    Skeletal Mesh=$User.SkeletalMesh
    Write Sampled Mesh Attrributes to Particle Attributes=true
   )
   &ParticleState()
   /Niagara/Modules/Update/Orientation/V2/UpdateMeshOrientation(
    Rotation Coordinate Space:ENiagaraMeshTransforms=::NewEnumerator3
    Rotation Rate:NiagaraFloat=$Particles.InitialRotationRate
    Rotation Vector:Vector3f=<1 0 0>
   )
   &ScaleColor(
    Scale Alpha=&FloatFromCurve(FloatCurve=DataInterface() {propertyValues {Curve {keys [
         {interpMode=RCIM_Cubic, tangentMode=RCTM_User, time=0, value=0}
         <at 0.1 1>
         {interpMode=RCIM_Cubic, tangentMode=RCTM_User, time=0.9, value=1}
         {arriveTangent=0.0513, interpMode=RCIM_Cubic, leaveTangent=0.0513, tangentMode=RCTM_User, time=1, value=0}
        ]}}})
    ScaleA=true
    ScaleRGB=true
   ) {module=ScaleColor001}
   &ScaleMeshSize(Scale Factor=&VectorFromFloat(Value=&FloatFromCurve(
      FloatCurve=DataInterface() {propertyValues {Curve {keys [
          {arriveTangent=5.9032, interpMode=RCIM_Cubic, tangentMode=RCTM_User, time=0, value=-0.0124}
          {arriveTangent=0.0124, interpMode=RCIM_Cubic, leaveTangent=0.0124, time=0.5341, value=1}
          {interpMode=RCIM_Cubic, leaveTangent=-2.4137, time=1, value=0}
         ]}}}
      Scale Curve=$Particles.ScaleFlipper
     )))
  ]
  rendererClasses [&NiagaraMeshRendererProperties]
  renderers [{rendererClass=&NiagaraMeshRendererProperties, rendererIndex=0}]
 }
 FN_Light {
  emitterSpawnScript [&FN_PreserveSkydiveUserVariables() {module=PreserveUserVariables}]
  emitterUpdateScript [
   &EmitterState()
   /Game/Effects/Niagara/Modules/Spawn/SpawnRateFromMultipleSpawnGroups_4(Global Spawnrate=2)
  ]
  particleSpawnScript [
   &FN_SetSpawnGroup_v2(Number of Spawn Groups=4)
   &InitializeParticle(
    Lifetime=0.7
    Module.Write Color=true
    Module.Write Lifetime=true
    Module.Write Mass=true
    Module.Write Material Random=true
    Module.Write Position=false
    Module.Write Scale=false
    Module.Write SpriteRotation=false
    Module.Write SpriteSize=true
    Position=/Niagara/Functions/Localspace/SimulationPosition()
    Sprite Size=&Vector2DFromFloat(Value=&UniformRangedFloat(Minimum=40, Maximum=80))
   )
  ]
  particleUpdateScript [
   &FN_Light_v3(
    Color=<rgba 25 15.1042 1.25 1>
    Create Color Attribute=true
    Create Exponent Attribute=true
    Create Radius Attribute=true
    Exponent=2
    Radius:NiagaraFloat=&MaskFloatBySpawnGroup(
     Masked Float=30
     Passthrough Float=&MaskFloatBySpawnGroup(Masked Float=30, Passthrough Float=50)
     Spawn Group Mask=1
    )
    Scale Color:NiagaraFloat=&FloatFromCurve(FloatCurve=DataInterface() {propertyValues {Curve {keys [
         {arriveTangent=3.5117, interpMode=RCIM_Cubic, leaveTangent=3.5117, tangentMode=RCTM_User, time=0, value=0}
         {interpMode=RCIM_Cubic, tangentMode=RCTM_User, time=0.5, value=1}
         {arriveTangent=-2.9519, interpMode=RCIM_Cubic, leaveTangent=-2.9519, tangentMode=RCTM_User, time=1, value=0}
        ]}}})
   )
   &SkeletalMeshLocation(
    Bone / Socket Index=$Particles.SpawnGroup
    Inherit Velocity=false
    MaskSpawnGroup=false
    Mesh Sampling Type=::NewEnumerator5
    Sampled Position Offset=&MaskVectorBySpawnGroup(
     Masked Vector=<0 0 0>
     Passthrough Vector=&MaskVectorBySpawnGroup(Masked Vector=<0 0 0>, Passthrough Vector=<0 0 10>)
     Spawn Group Mask=1
    )
    Skeletal Mesh=$User.SkeletalMesh
    Socket Sampling Mode=::NewEnumerator5
   )
   &ParticleState()
  ]
  rendererClasses [&NiagaraLightRendererProperties]
  renderers [{rendererClass=&NiagaraLightRendererProperties, rendererIndex=0}]
 }
 FN_Mesh_GrapeEnergy {
  emitterSpawnScript [&FN_PreserveSkydiveUserVariables() {module=PreserveUserVariables}]
  emitterUpdateScript [
   &EmitterState()
   &SpawnRateFromMultipleSpawnGroups_4(Global Spawnrate:NiagaraFloat=&UniformRangedFloat(Minimum=1, Maximum=8))
   /Game/Effects/Niagara/Modules/Skydiving/FN_DivingAttributes/FN_EmitterDivingAttributes_v1(SkeletalMesh:NiagaraDataInterfaceSkeletalMesh=$User.SkeletalMesh)
  ]
  particleSpawnScript [
   &InitializeParticle(
    Lifetime=&UniformRangedFloat(Minimum=1, Maximum=1)
    Mesh Scale=&Multiply_Vector(
     A=<1 1 1>
     B=&VectorFromFloat(Value=&UniformRangedFloat(Minimum=0.6, Maximum=0.8))
    )
    Module.Write Color=false
    Module.Write Lifetime=true
    Module.Write Mass=true
    Module.Write Material Random=true
    Module.Write Position=false
    Module.Write Scale=true
    Module.Write SpriteRotation=true
    Module.Write SpriteSize=true
    Position=&SimulationPosition()
    Sprite Rotation=&UniformAOrBFloat(A=0, B=360)
    Sprite Size=&UniformRangedVector2D(Minimum=<20 20>, Maximum=<25 25>)
   )
   /SetVariables(Particles.Mesh-InitialRotationVector:Vector3f=&UniformRangedVector(Minimum=<-0.15 -0.15 0.65>, Maximum=<0.15 0.15 1>)) {isSetParameters=true}
   /SetVariables(Particles.Mesh-InitialRotationRate:NiagaraFloat=&UniformAOrBFloat(A=-1.5, B=1.5)) {isSetParameters=true}
   /SetVariables(Particles.Color=&MakeLinearColorFromVectorAndFloat(
     "Vector (RGB)"=&MakeVectorFromLinearColorRGB(Color=&ColorFromCurve(ColorCurve=DataInterface() {propertyValues {
         AlphaCurve {keys [<at -0.0023 1 RCIM_Cubic>,<at 0.9977 1 RCIM_Cubic>]}
         BlueCurve {keys [<at 0 0.02 RCIM_Cubic>]}
         GreenCurve {keys [<at 0 0.6407 RCIM_Cubic>]}
         RedCurve {keys [<at 0 1 RCIM_Cubic>]}
        }}))
     "Float (Alpha)"=0.5
    )) {isSetParameters=true, module=SetVariables_#1}
   /SetVariables(Particles.Color_Mesh:LinearColor=&MakeLinearColorFromVectorAndFloat(
     "Vector (RGB)"=&MakeVectorFromLinearColorRGB(Color=&ColorFromCurve(ColorCurve=DataInterface() {propertyValues {
         AlphaCurve {keys [<at -0.0023 1 RCIM_Cubic>,<at 0.9977 1 RCIM_Cubic>]}
         BlueCurve {keys [<at 0 0.225 RCIM_Cubic>]}
         GreenCurve {keys [<at 0 0.8625 RCIM_Cubic>]}
         RedCurve {keys [<at 0 1.5 RCIM_Cubic>]}
        }}))
     "Float (Alpha)"=1
    )) {isSetParameters=true}
   &FN_SetSpawnGroup_v2(Number of Spawn Groups=4)
   &SkeletalMeshLocation(
    Bone / Socket Index=$Particles.SpawnGroup
    Bone Sampling Mode=::NewEnumerator4
    Inherit Velocity=false
    MaskSpawnGroup=false
    Skeletal Mesh=$User.SkeletalMesh
   )
   &AddVelocity(
    Scale Added Velocity=&Lerp_Vector(A=<1 1 1>, B=<1 1 3>, Alpha=$Emitter.Dive)
    Velocity=&UniformRangedVector(Minimum=<0 0 100>, Maximum=<1 1 200>)
   )
   /Niagara/Modules/Spawn/Location/SphereLocation(MaskSpawnGroup=false, Module.Override Randomness=false, Sphere Radius=20, Surface Only=false)
   &InitialMeshOrientation(
    Orientation Coordinate Space:ENiagaraCoordinateSpace=::World
    Orientation Vector=&UniformRangedVector(Minimum=<0 0 0>, Maximum=<360 0 0>)
    Rotation Coordinate Space:ENiagaraMeshTransforms=::NewEnumerator0
    Use Orientation Vector=true
    Use Rotation Vector=true
   )
  ]
  particleUpdateScript [
   &ParticleState()
   &Drag(Drag=0.5, Use Linear Drag=true, Use Rotational Drag=false)
   /Niagara/Modules/Update/Forces/CurlNoiseForce(
    Noise Frequency:NiagaraFloat=&UniformRangedFloat(Minimum=0.5, Maximum=1)
    Noise Strength=&UniformRangedFloat(Minimum=300, Maximum=500)
    Pan Noise=false
    Pan Noise Field=<1 1 0>
   )
   &SolveForcesAndVelocity(Clamp Velocity=false, Limit Acceleration=false)
   &UpdateMeshOrientation(
    Rotation Rate:NiagaraFloat=Linked() {linkedVariable {name=Particles.Mesh-InitialRotationRate, type=NiagaraFloat}}
    Rotation Vector:Vector3f=Linked() {linkedVariable {name=Particles.Mesh-InitialRotationVector, type=Vector3f}}
   )
   &ScaleMeshSize(Scale Factor=&VectorFromFloat(Value=&FloatFromCurve(FloatCurve=DataInterface() {propertyValues {Curve {keys [
          {arriveTangent=6.5937, interpMode=RCIM_Cubic, leaveTangent=6.5937, tangentMode=RCTM_User, time=0, value=0}
          <at 0.3 1 RCIM_Cubic>
          {arriveTangent=-3.6524, interpMode=RCIM_Cubic, leaveTangent=-3.6524, tangentMode=RCTM_User, time=1, value=0}
         ]}}})))
   &ScaleSpriteSize(Scale Factor=&Vector2DFromFloat(Value=&FloatFromCurve(FloatCurve=DataInterface() {propertyValues {Curve {keys [
          {arriveTangent=4.2717, interpMode=RCIM_Cubic, leaveTangent=4.2717, tangentMode=RCTM_User, time=0, value=0}
          <at 0.3 1 RCIM_Cubic>
          {arriveTangent=-3.6524, interpMode=RCIM_Cubic, leaveTangent=-3.6524, tangentMode=RCTM_User, time=1, value=0}
         ]}}})))
  ]
  rendererClasses [&NiagaraSpriteRendererProperties, &NiagaraMeshRendererProperties]
  renderers [
   {rendererClass=&NiagaraSpriteRendererProperties}
   {rendererClass=&NiagaraMeshRendererProperties, rendererIndex=1}
  ]
 }
 FN_Ribbons {
  emitterUpdateScript [
   &EmitterState()
   &FN_EmitterDivingAttributes_v1(SkeletalMesh:NiagaraDataInterfaceSkeletalMesh=$User.SkeletalMesh)
   &FN_SpawnRibbon_v2(Number of Ribbons=4, Spawn Mode=::NewEnumerator0, SpawnRate=18)
  ]
  particleSpawnScript [
   &FN_InitializeRibbon_v3(
    Color:LinearColor=/Niagara/DynamicInputs/TypeConversions/MakeLinearColorFromFloat(Value:NiagaraFloat=0.6)
    Enable Skeletal Position=true
    Lifetime=0.35
    Ribbon Width=50
    Sampled Position Coordinate Space:ENiagaraCoordinateSpace=::Local
    Sampled Rotation Coordinate Space:ENiagaraCoordinateSpace=::Local
    Sampled Velocity Coordinate Space:ENiagaraCoordinateSpace=::Local
   ) {module=FN_InitializeRibbon_v3001}
   &AddVelocity(
    Coordinate Space=::World
    Scale Added Velocity=&Lerp_Vector(A=<1 1 1>, B=<1 1 2>, Alpha=$Emitter.Dive)
    Velocity=<0 0 500>
   )
   &AddVelocity(Velocity=&MakeCustomVectorFromBool(
     Bool=$Particles.Locked
     True Vector=<0 0 0>
     False Vector=&FN_FrontEndVector3Modifier(Gameplay=<0 0 500>, Front End=<0 0 500>)
    )) {module=AddVelocity001}
   &SampleSkeletalMeshSkeleton(Mesh:NiagaraDataInterfaceSkeletalMesh=$User.SkeletalMesh)
  ]
  particleUpdateScript [
   &SolveForcesAndVelocity(Clamp Velocity=false, Limit Acceleration=false)
   &FN_UpdateRibbon(
    Enable Skeletal Position=true
    Initial Color:LinearColor=$Particles.Initial.Color
    Initial RibbonWidth:NiagaraFloat=$Particles.Initial.RibbonWidth
   )
   &Color(
    Color=&ColorFromCurve(ColorCurve=DataInterface() {propertyValues {
       AlphaCurve {keys [<at 0 1>]}
       BlueCurve {keys [<at 0 0.172>, <at 0.1935 0.0359>, {arriveTangent=-0.3652, interpMode=RCIM_Cubic, leaveTangent=-0.3652, time=0.2445, value=0}, <at 0.4709 0>]}
       GreenCurve {keys [
         <at 0 0.749>
         <at 0.1935 0.6179>
         {arriveTangent=-0.7057, interpMode=RCIM_Cubic, leaveTangent=-0.7057, time=0.2445, value=0.5833}
         <at 0.4709 0.4167>
        ]}
       RedCurve {keys [<at 0 1>,<at 0.1935 1>,<at 0.2445 1 RCIM_Cubic>,<at 0.4709 1>]}
      }})
    Scale Alpha=&FloatFromCurve(FloatCurve=DataInterface() {propertyValues {Curve {keys [
         {arriveTangent=0.7543, interpMode=RCIM_Cubic, leaveTangent=0.7543, tangentMode=RCTM_User, time=0, value=0}
         <at 0.1 0.05>
         {arriveTangent=-0.013, interpMode=RCIM_Cubic, leaveTangent=-0.013, tangentMode=RCTM_User, time=0.2992, value=0.0096}
         {interpMode=RCIM_Cubic, leaveTangent=-0.2611, time=1, value=0}
        ]}}})
   )
   &Material(Base Texture Time:NiagaraFloat=&Multiply_Float(A=$Emitter.Age, B=-1), PostMult=0.5455, Power=0.4408)
  ]
  rendererClasses [&NiagaraRibbonRendererProperties]
  renderers [{rendererClass=&NiagaraRibbonRendererProperties}]
 }
 FN_SourceGlows {
  emitterSpawnScript [&FN_PreserveSkydiveUserVariables() {module=PreserveUserVariables}]
  emitterUpdateScript [
   &EmitterState()
   &SpawnRateFromMultipleSpawnGroups_4(Global Spawnrate=2)
  ]
  particleSpawnScript [
   &FN_SetSpawnGroup_v2(Number of Spawn Groups=4)
   &InitializeParticle(
    Lifetime=&UniformRangedFloat(Minimum=0.5, Maximum=0.65)
    Module.Write Color=true
    Module.Write Lifetime=true
    Module.Write Mass=true
    Module.Write Material Random=true
    Module.Write Position=false
    Module.Write Scale=false
    Module.Write SpriteRotation=false
    Module.Write SpriteSize=true
    Position=&SimulationPosition()
    Sprite Size=&Vector2DFromFloat(Value=&UniformRangedFloat(Minimum=70, Maximum=80))
   )
   &CameraOffset(Camera Offset Amount=&MaskFloatBySpawnGroup(
     Masked Float=20
     Passthrough Float=&MaskFloatBySpawnGroup(Masked Float=20, Passthrough Float=25)
     Spawn Group Mask=1
    ))
  ]
  particleUpdateScript [
   /SetVariables(Particles.Color=&MakeLinearColorFromVectorAndFloat(
     "Vector (RGB)"=&MakeVectorFromLinearColorRGB(Color=&ColorFromCurve(ColorCurve=DataInterface() {propertyValues {
         AlphaCurve {keys [<at -0.0023 1 RCIM_Cubic>,<at 0.9977 1 RCIM_Cubic>]}
         BlueCurve {keys [<at 0 0.1>,<at 0.341 0.15>,<at 1 0.1 RCIM_Cubic>]}
         GreenCurve {keys [<at 0 0.625>,<at 0.341 0.7167>,<at 1 0.625 RCIM_Cubic>]}
         RedCurve {keys [<at 0 1>,<at 0.341 1>,<at 1 1 RCIM_Cubic>]}
        }}))
     "Float (Alpha)"=&FloatFromCurve(FloatCurve=DataInterface() {propertyValues {Curve {keys [
          {arriveTangent=1.1784, interpMode=RCIM_Cubic, leaveTangent=1.1784, tangentMode=RCTM_User, time=0, value=0}
          {arriveTangent=0.002, interpMode=RCIM_Cubic, leaveTangent=0.002, tangentMode=RCTM_User, time=0.5, value=0.2}
          {arriveTangent=-1.0701, interpMode=RCIM_Cubic, leaveTangent=-1.0701, tangentMode=RCTM_User, time=1.0011, value=0.002}
         ]}}})
    )) {isSetParameters=true}
   &SkeletalMeshLocation(
    Bone / Socket Index=$Particles.SpawnGroup
    Bone Sampling Mode=::NewEnumerator4
    Inherit Velocity=false
    MaskSpawnGroup=false
    Skeletal Mesh=$User.SkeletalMesh
   )
   &ParticleState()
   &ScaleMeshSize(Scale Factor=&Multiply_Vector(
     A=<1 1 1>
     B=&VectorFromFloat(Value=&FloatFromCurve(FloatCurve=DataInterface() {propertyValues {Curve {keys [<at 0 0>,<at 1 1.5>]}}}))
    ))
  ]
  rendererClasses [&NiagaraSpriteRendererProperties]
  renderers [{rendererClass=&NiagaraSpriteRendererProperties}]
 }
 FN_SparkleDust {
  emitterSpawnScript [&FN_PreserveSkydiveUserVariables() {module=PreserveUserVariables}]
  emitterUpdateScript [
   &EmitterState()
   &SpawnRateFromMultipleSpawnGroups_4(Global Spawnrate=16)
   &FN_EmitterDivingAttributes_v1(SkeletalMesh:NiagaraDataInterfaceSkeletalMesh=$User.SkeletalMesh)
  ]
  particleSpawnScript [
   &FN_SetSpawnGroup_v2(Number of Spawn Groups=4)
   &InitializeParticle(
    Lifetime=&UniformRangedFloat(Minimum=0.6, Maximum=0.8)
    Mesh Scale=&Multiply_Vector(
     A=<1 1 1>
     B=&VectorFromFloat(Value=&UniformRangedFloat(Minimum=0.2, Maximum=1))
    )
    Module.Write Color=false
    Module.Write Lifetime=true
    Module.Write Mass=true
    Module.Write Material Random=true
    Module.Write Position=false
    Module.Write Scale=false
    Module.Write SpriteRotation=true
    Module.Write SpriteSize=true
    Position=&SimulationPosition()
    Sprite Rotation=&UniformAOrBFloat(A=0, B=360)
    Sprite Size=&Vector2DFromFloat(Value=&UniformRangedFloat(Minimum=3, Maximum=5))
   )
   &SkeletalMeshLocation(
    Bone / Socket Index=$Particles.SpawnGroup
    Bone Sampling Mode=::NewEnumerator4
    Inherit Velocity=false
    MaskSpawnGroup=false
    Skeletal Mesh=$User.SkeletalMesh
   )
   &AddVelocity(
    Coordinate Space=::World
    Scale Added Velocity=&Lerp_Vector(A=<1 1 1>, B=<1 1 3>, Alpha=$Emitter.Dive)
    Velocity=&UniformRangedVector(Minimum=<0 0 125>, Maximum=<1 1 225>)
   )
   &SphereLocation(MaskSpawnGroup=false, Module.Override Randomness=false, Sphere Radius=12.5, Surface Only=false)
  ]
  particleUpdateScript [
   &ParticleState()
   &Drag(Drag=0.5, Use Linear Drag=true, Use Rotational Drag=false)
   &CurlNoiseForce(
    Noise Frequency:NiagaraFloat=&UniformRangedFloat(Minimum=0.5, Maximum=1)
    Noise Strength=&UniformRangedFloat(Minimum=300, Maximum=500)
    Pan Noise=false
    Pan Noise Field=<1 1 0>
   )
   &SolveForcesAndVelocity(Clamp Velocity=false, Limit Acceleration=false)
   &ScaleSpriteSize(Scale Factor=&Multiply_Vector2(
     A=<1 1>
     B=&Vector2DFromFloat(Value=&FloatFromCurve(FloatCurve=DataInterface() {propertyValues {Curve {keys [<at -0.0046 3>,<at 1.0011 0.002>]}}}))
    ))
   /SetVariables(Particles.Color=&MakeLinearColorFromVectorAndFloat(
     "Vector (RGB)"=&MakeVectorFromLinearColorRGB(Color=&ColorFromCurve(ColorCurve=DataInterface() {propertyValues {
         AlphaCurve {keys [<at -0.0023 1 RCIM_Cubic>,<at 0.9977 1 RCIM_Cubic>]}
         BlueCurve {keys [<at 0 0.1 RCIM_Cubic>,<at 1 0>]}
         GreenCurve {keys [<at 0 0.67 RCIM_Cubic>,<at 1 0.25>]}
         RedCurve {keys [<at 0 1 RCIM_Cubic>,<at 1 0.75>]}
        }}))
     "Float (Alpha)"=&FloatFromCurve(FloatCurve=DataInterface() {propertyValues {Curve {keys [<at 0 0>,<at 0.1936 0.8315>,<at 0.4019 0.1015>,<at 0.5589 0.6516>,<at 0.7294 0.0749>,<at 0.8673 0.41>,<at 1.0011 0.002>]}}})
    )) {isSetParameters=true, module=SetVariables_#1}
  ]
  rendererClasses [&NiagaraSpriteRendererProperties]
  renderers [{rendererClass=&NiagaraSpriteRendererProperties}]
 }
}
summary {
 systemName=NS_Contrail_Hightower_Grape
 emitters [
  {emitterName=FN_Ribbons, rendererClasses [&NiagaraRibbonRendererProperties]}
  {emitterName=FN_SparkleDust, rendererClasses [&NiagaraSpriteRendererProperties]}
  {emitterName=FN_SourceGlows, rendererClasses [&NiagaraSpriteRendererProperties]}
  {
   emitterName=FN_Mesh_GrapeEnergy
   rendererClasses [&NiagaraSpriteRendererProperties, &NiagaraMeshRendererProperties]
  }
  {emitterName=FN_Light, rendererClasses [&NiagaraLightRendererProperties]}
  {emitterName=Brambles001, rendererClasses [&NiagaraMeshRendererProperties]}
 ]
 userVariables [
  User.bIsFrontEnd {type=NiagaraBool}
  User.bIsFrontEndPreview {type=NiagaraBool}
  User.PlayerSpeed {type=NiagaraFloat, defaultValue=0, description=""}
  User.RibbonIDOffset {type=NiagaraInt32}
  User.SkeletalMesh {
   type=NiagaraDataInterfaceSkeletalMesh
   defaultValue {
    struct=NiagaraExt_VariableValue_DataInterface
    value {dataInterface=&self:NiagaraDataInterfaceSkeletalMesh_0}
   }
  }
 ]
}
